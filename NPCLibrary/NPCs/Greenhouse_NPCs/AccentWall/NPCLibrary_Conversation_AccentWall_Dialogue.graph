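# Conversation database for the AccentWall greenhouse NPC in the NPC library, holding its dialogue about Calypso.
# A single entry tag opens a taunting greeting under an "Ask about..." choice, which branches to a Calypso topic or a Back option.
# The Calypso topic plays a linear chain of four Speech nodes telling of duty and betrayal, then a ReturnToStart node offers BACK.
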ConversationDatabase /NPCLibrary/NPCs/Greenhouse_NPCs/AccentWall/NPCLibrary_Conversation_AccentWall_Dialogue
entry_points [{tag=Conversation.Entry.AccentWall.Dialogue, destinations [#1]}]
nodes [
 ConversationEntryPointNode_0 {
  id=#1
  class=ConversationEntryPointNode
  outputs [#2]
  properties {entryTag {tagName=Conversation.Entry.AccentWall.Dialogue}}
 }
 Speech_C_0 {
  id=#2
  class=Speech_C
  outputs [#3, #4]
  properties {nodeName="", softTaskIcon=&T_UI_NPC_question, generalConfig {defaultMessage="Ha! You afraid to get wet?"}}
  sub_nodes [ConversationChoiceNode_5 {class=ConversationChoiceNode, properties {defaultChoiceDisplayText=Ask about...}}]
 }
 Back_C_0 {
  id=#4
  class=Back_C
  sub_nodes [ConversationChoiceNode_7 {class=ConversationChoiceNode, properties {defaultChoiceDisplayText=BACK}}]
 }
 Speech_C_7 {
  id=#3
  class=Speech_C
  outputs [#5]
  properties {nodeName="", softTaskIcon=&T_UI_NPC_question, generalConfig {defaultMessage=She charged me with a duty.}}
  sub_nodes [ConversationChoiceNode_15 {class=ConversationChoiceNode, properties {defaultChoiceDisplayText=Calypso}}]
 }
 Speech_C_15 {
  id=#5
  class=Speech_C
  outputs [#6]
  properties {
   nodeName=""
   softTaskIcon=&T_UI_NPC_question
   generalConfig {defaultMessage=For ten years I looked after those who were lost to the sea.}
  }
 }
 Speech_C_14 {
  id=#6
  class=Speech_C
  outputs [#7]
  properties {
   nodeName=""
   softTaskIcon=&T_UI_NPC_question
   generalConfig {defaultMessage="And finally, when we could be together again..."}
  }
 }
 Speech_C_1 {
  id=#7
  class=Speech_C
  outputs [#8]
  properties {nodeName="", softTaskIcon=&T_UI_NPC_question, generalConfig {defaultMessage=She. Wasn't. There.}}
 }
 ReturnToStart_C_23 {
  id=#8
  class=ReturnToStart_C
  properties {softTaskIcon=&T_UI_NPC_Back}
  sub_nodes [ConversationChoiceNode_32 {class=ConversationChoiceNode, properties {defaultChoiceDisplayText=BACK}}]
 }
]
