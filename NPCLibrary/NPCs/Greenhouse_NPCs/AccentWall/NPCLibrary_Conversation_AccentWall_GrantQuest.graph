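# Conversation database for the AccentWall Greenhouse NPC that offers and grants the "Survive the Black Mark" urgent quest.
# The entry tag leads to a speech choice, which branches to Confirm or Back.
# Confirm grants the static Greenhouse Urgent 01 quest and records a GrantedQuest history event server-side.
# Both the offer and Confirm use an inverted Has Urgent Quest requirement, blocking them while the player already holds that urgent quest.
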ConversationDatabase /NPCLibrary/NPCs/Greenhouse_NPCs/AccentWall/NPCLibrary_Conversation_AccentWall_GrantQuest
entry_points [{tag=Conversation.Entry.AccentWall.GrantQuest, destinations [#1]}]
nodes [
 ConversationEntryPointNode_0 {
  id=#1
  class=ConversationEntryPointNode
  outputs [#2]
  properties {entryTag {tagName=Conversation.Entry.AccentWall.GrantQuest}}
 }
 Speech_C_0 {
  id=#2
  class=Speech_C
  outputs [#3, #4]
  properties {
   nodeName=""
   softTaskIcon=/Game/Athena/UI/Challenges/Art/ChallengeTypeIcons/UI_ChallengeIcon_Visit
   generalConfig {defaultMessage=Survive the Black Mark}
  }
  sub_nodes [
   ConversationChoiceNode_5 {class=ConversationChoiceNode, properties {defaultChoiceDisplayText=Survive the Black Mark}}
   FortConversationRequirement_ControllerRequirement_0 {
    class=FortConversationRequirement_ControllerRequirement
    properties {
     nodeName=Require: Has Urgent Quest
     requirement {
      class=FortControllerRequirement_HasUrgentQuest
      properties {bInverseCondition=true, acceptedUrgentEventTags {gameplayTags [{tagName=ChallengeCategory.S30.EE.Greenhouse.Urgent.01}]}}
     }
    }
   }
  ]
 }
 Back_C_4 {
  id=#4
  class=Back_C
  sub_nodes [ConversationChoiceNode_3 {class=ConversationChoiceNode, properties {defaultChoiceDisplayText=Back}}]
 }
 GrantStaticQuest_C_0 {
  id=#3
  class=GrantStaticQuest_C
  outputs []
  properties {
   bSharedQuestLimitsEnforced=false
   bShouldOverrideSpeakerText=false
   questToGrantPtr=/BattlePassPermanentQuests/Items/QuestItems/AccentWall/Quest_S30_EE_Greenhouse_Urgent_01
   softTaskIcon=/Game/UI/Foundation/Textures/Icons/Checks/T-Icon-Check-128
  }
  sub_nodes [
   ConversationChoiceNode_2 {class=ConversationChoiceNode, properties {defaultChoiceDisplayText=Confirm}}
   SideEffect_Server_RecordEventToHistoryData_C_1 {class=SideEffect_Server_RecordEventToHistoryData_C, properties {event Type {tagName=Conversation.Event.GrantedQuest}}}
   FortConversationRequirement_ControllerRequirement_1 {
    class=FortConversationRequirement_ControllerRequirement
    properties {
     nodeName=Require: Has Urgent Quest
     requirement {
      class=FortControllerRequirement_HasUrgentQuest
      properties {bInverseCondition=true, acceptedUrgentEventTags {gameplayTags [{tagName=ChallengeCategory.S30.EE.Greenhouse.Urgent.01}]}}
     }
    }
   }
  ]
 }
]
